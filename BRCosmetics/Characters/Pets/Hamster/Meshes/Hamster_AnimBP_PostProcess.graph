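# Post-process AnimBlueprint for the Hamster pet, targeting the shared Chameleon skeleton, that adds secondary motion and corrective bone rotation.
# The input pose runs through three serial Anim Dynamics nodes that simulate dyn_hair_01 to 03 as angular-limited springs in actor space at partial alpha.
# Two Constraint nodes then drive each thigh_OL bone's rotation from the pelvis and matching thigh with equal weights.
# The EventGraph holds no logic.
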AnimBlueprint /BRCosmetics/Characters/Pets/Hamster/Meshes/Hamster_AnimBP_PostProcess
target_skeleton=/Game/Characters/Pets/Chameleon/Meshes/Chameleon_Skeleton
graphs [
 AnimGraph {nodes [
   AnimGraphNode_SubInput_0 = Input Pose()
   n1 = Anim Dynamics - Bone: dyn_hair_01(ComponentPose=Local To Component(LocalPose=@AnimGraphNode_SubInput_0), Alpha=0.7) {
    angularDampingOverride=0.97
    angularSpringConstant=1000
    bAngularSpring=true
    bGravityOverrideInSimSpace=true
    bOverrideAngularDamping=true
    bUsePlanarLimit=false
    simulationSpace=Actor
    alphaBoolBlend {blendOption=Linear}
    boundBone {boneName=dyn_hair_01}
    physicsBodyDefinitions [{
      sphereCollisionRadius=0
      boundBone {boneName=dyn_hair_01}
      boxExtents=<5 5 2.5>
      constraintSetup {angularTargetAxis=AxisZ, twistAxis=AxisZ, angularLimitsMax=<15 15 0>, angularLimitsMin=<-15 -15 0>, angularTarget=<0 0 0>}
      localJointOffset=<0 0 -0.1>
     }]
   }
   n2 = Anim Dynamics - Bone: dyn_hair_02(ComponentPose=@n1, Alpha=0.7) {
    angularDampingOverride=0.97
    angularSpringConstant=1000
    bAngularSpring=true
    bGravityOverrideInSimSpace=true
    bOverrideAngularDamping=true
    bUsePlanarLimit=false
    simulationSpace=Actor
    alphaBoolBlend {blendOption=Linear}
    boundBone {boneName=dyn_hair_02}
    physicsBodyDefinitions [{
      sphereCollisionRadius=0
      boundBone {boneName=dyn_hair_02}
      boxExtents=<5 5 2.5>
      constraintSetup {angularTargetAxis=AxisZ, twistAxis=AxisZ, angularLimitsMax=<15 15 0>, angularLimitsMin=<-15 -15 0>, angularTarget=<0 0 0>}
      localJointOffset=<0 0 -0.1>
     }]
   }
   n3 = Anim Dynamics - Bone: dyn_hair_03(ComponentPose=@n2, Alpha=0.7, ExternalForce="0,0,0") {
    angularDampingOverride=0.97
    angularSpringConstant=1000
    bAngularSpring=true
    bGravityOverrideInSimSpace=true
    bOverrideAngularDamping=true
    bUsePlanarLimit=false
    simulationSpace=Actor
    alphaBoolBlend {blendOption=Linear}
    boundBone {boneName=dyn_hair_03}
    physicsBodyDefinitions [{
      sphereCollisionRadius=0
      boundBone {boneName=dyn_hair_03}
      boxExtents=<5 5 2.5>
      constraintSetup {angularTargetAxis=AxisZ, twistAxis=AxisZ, angularLimitsMax=<15 15 0>, angularLimitsMin=<-15 -15 0>, angularTarget=<0 0 0>}
      localJointOffset=<0 0 -0.1>
     }]
   }
   n4 = "Constraint - thigh_OL_l (RR)"!(ComponentPose=@n3, Alpha=1, ConstraintWeights_0=1, ConstraintWeights_1:float=1) -> (Pose) {
    alphaBoolBlend {blendOption=Linear}
    boneToModify {boneName=thigh_OL_l}
    constraintSetup [{transformType=Rotation, targetBone {boneName=pelvis}}, {transformType=Rotation, targetBone {boneName=thigh_l}}]
    constraintWeights [1,1]
   }
   n5 = "Constraint - thigh_OL_r (RR)"!(ComponentPose=@n4, Alpha=1, ConstraintWeights_0=1, ConstraintWeights_1:float=1) -> (Pose) {
    alphaBoolBlend {blendOption=Linear}
    boneToModify {boneName=thigh_OL_r}
    constraintSetup [{transformType=Rotation, targetBone {boneName=pelvis}}, {transformType=Rotation, targetBone {boneName=thigh_r}}]
    constraintWeights [1,1]
   }
   Output Pose(Result=Component To Local(ComponentPose=@n5)) {layerGroup=None}
  ]}
 EventGraph {nodes [
   TryGetPawnOwner()
   Event BlueprintUpdateAnimation()
  ]}
]
variables []
